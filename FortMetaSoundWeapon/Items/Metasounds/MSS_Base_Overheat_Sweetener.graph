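# Mono one-shot MetaSound source for a weapon overheat sweetener layer, using the weapon foley sound class and attenuation.
# Play or ReTrigger plays a random wave from Sounds, pitched by OverheatPitch converted to semitones plus a small offset and random spread.
# A randomized, zero-floored gain multiplied by OverheatVolume scales the output, silent by default.
# A deprecated weapon-fire subgraph tracks completion, and OnFinished fires after a short delay.
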MetaSoundSource /FortMetaSoundWeapon/Items/Metasounds/MSS_Base_Overheat_Sweetener
inputs [
 OverheatPitch {type=Float}
 OverheatVolume {type=Float}
 Random.Pitch.Base {type=Float}
 Random.Pitch.Spread {type=Float}
 Random.Volume.Base {type=Float}
 Random.Volume.Spread {type=Float}
 ReTrigger {type=Trigger}
 Sounds {type=WaveAsset:Array}
 UE.Source.OnPlay {type=Trigger}
]
interfaces [UE.OutputFormat.Mono:1.0, UE.Source.OneShot:1.0, UE.Source:1.1]
outputs [
 bOnAlmostFinished {type=Bool}
 "UE.OutputFormat.Mono.Audio:0" {type=Audio}
 UE.Source.OneShot.OnFinished {type=Trigger}
]
pages [Default {
  input_defaults {
   OverheatPitch=1
   OverheatVolume=0
   Random.Pitch.Base=0
   Random.Pitch.Spread=0
   Random.Volume.Base=1.2
   Random.Volume.Spread=0
   ReTrigger=null
   UE.Source.OnPlay=null
   Sounds []
  }
  nodes [
   TriggerAnyTriggerAny2_0 = "TriggerAny.Trigger Any (2)"(In 1=^ReTrigger, In 0=^UE.Source.OnPlay)
   n1 = UE.RandomFloat(
    Max=UE.Add:Float(PrimaryOperand=^Random.Volume.Base, AdditionalOperands=^Random.Volume.Spread).Out
    Next=UE.Reroute(Value=@TriggerAnyTriggerAny2_0.Out).Value
    Min=UE.Subtract:Float(PrimaryOperand=^Random.Volume.Base, AdditionalOperands=^Random.Volume.Spread).Out
   )
   FrequencyMultipliertoSemitone_0 = UE.Frequency Multiplier to Semitone:Float(Frequency Multiplier=^OverheatPitch)
   n2 = UE.Add:Float(PrimaryOperand=^Random.Pitch.Base, AdditionalOperands=^Random.Pitch.Spread)
   n3 = UE.Reroute(Value=@TriggerAnyTriggerAny2_0.Out)
   n4 = UE.Subtract:Float(PrimaryOperand=^Random.Pitch.Base, AdditionalOperands=^Random.Pitch.Spread)
   ArrayRandomGet_0 = Array.Random Get:WaveAsset:Array(In Array=^Sounds, Next=@TriggerAnyTriggerAny2_0.Out)
   WavePlayer_0 = UE.Wave Player:Mono(
    Pitch Shift=UE.Add:Float(
     PrimaryOperand=UE.Add:Float(PrimaryOperand=@FrequencyMultipliertoSemitone_0.Semitones) {also {defaults {AdditionalOperands=0.28}}}.Out
     AdditionalOperands=UE.RandomFloat(Max=@n2.Out, Next=@n3.Value, Min=@n4.Out).Value
    ).Out
    Play=@ArrayRandomGet_0.On Next
    Wave Asset=@ArrayRandomGet_0.Value
    Stop=VariableAccessor.Trigger(Variable=InitVariable.Trigger().Variable).Value
   )
   Multiply_0 = UE.Multiply:Audio by Float(
    AdditionalOperands=UE.Multiply:Float(PrimaryOperand=^OverheatVolume, AdditionalOperands=Max.Max:Float(A=@n1.Value) {also {defaults {B=0}}}.Value).Out
    PrimaryOperand=@WavePlayer_0.Out Mono
   )
   n5 = InitVariable.Trigger()
   MSP_Weapon_Fire_OnFinished_0 = /FortMetaSoundWeapon/Items/Metasounds/Subgraphs/Deprecated/MSP_Weapon_Fire_OnFinished(
    ReTrigger=^ReTrigger
    UE.Source.OnPlay=^UE.Source.OnPlay
    SoundsFinished=VariableAccessor.Trigger(Variable=VariableMutator.Trigger(Variable=@n5.Variable, Value=@WavePlayer_0.On Finished).Variable).Value
   )
   TriggerDelay_0 = UE.Trigger Delay(In=@MSP_Weapon_Fire_OnFinished_0.UE.Source.OneShot.OnFinished) {also {defaults {Delay Time=0.5}}}
  ]
  links [
   {dst="@bOnAlmostFinished", src=MSP_Weapon_Fire_OnFinished_0.bOnAlmostFinished}
   {dst="@UE.OutputFormat.Mono.Audio:0", src=Multiply_0.Out}
   {dst="@UE.Source.OneShot.OnFinished", src=TriggerDelay_0.Out}
  ]
 }]
properties {
 assetImportData=&self:AssetImportData
 attenuationSettings=&AS_Fort_WeaponFoley
 maxDistance=2250
 sampleRate=48000
 soundClassObject=&Fort_WeaponFoley
}
