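# MetaSound preset of MSS_Base_Overheat_Sweetener that supplies the overheat sweetener layer sound for the Jethro CalmRock weapon.
# It feeds four CalmRock overheat layer waves into the parent's Sounds array, with a raised base volume and no pitch or volume random spread.
# Overheat pitch, overheat volume and retrigger inputs are exposed, and the mono one-shot output uses weapon foley attenuation and sound class.
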MetaSoundSource /JethroWeaponGameplay/Assets/Audio/CalmRock/MSS_CalmRock_Overheat_Sweetener_Preset
preset_parent=/FortMetaSoundWeapon/Items/Metasounds/MSS_Base_Overheat_Sweetener
inputs [
 OverheatPitch {type=Float}
 OverheatVolume {type=Float}
 Random.Pitch.Base {type=Float}
 Random.Pitch.Spread {type=Float}
 Random.Volume.Base {type=Float}
 Random.Volume.Spread {type=Float}
 ReTrigger {type=Trigger}
 Sounds {type=WaveAsset:Array}
 UE.Source.OnPlay {type=Trigger}
]
interfaces [UE.OutputFormat.Mono:1.0, UE.Source.OneShot:1.0, UE.Source:1.1]
outputs [
 bOnAlmostFinished {type=Bool}
 "UE.OutputFormat.Mono.Audio:0" {type=Audio}
 UE.Source.OneShot.OnFinished {type=Trigger}
]
pages [Default {input_defaults {
   OverheatPitch=1
   OverheatVolume=0
   Random.Pitch.Base=0
   Random.Pitch.Spread=0
   Random.Volume.Base=1.2
   Random.Volume.Spread=0
   ReTrigger=null
   UE.Source.OnPlay=null
   Sounds [
    &/Jethro_CalmRock_Overheat_Layer_04
    &/Jethro_CalmRock_Overheat_Layer_03
    &/Jethro_CalmRock_Overheat_Layer_02
    &/Jethro_CalmRock_Overheat_Layer_01
   ]
  }}]
properties {
 assetImportData=&self:AssetImportData
 attenuationSettings=&AS_Fort_WeaponFoley
 maxDistance=2250
 soundClassObject=&Fort_WeaponFoley
}
